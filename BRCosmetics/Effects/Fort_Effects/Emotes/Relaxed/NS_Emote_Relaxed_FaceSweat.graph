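# Niagara system for face sweat on the Relaxed emote, built from reused drool emitters.
# DroolPendulum hangs stretching mesh drips from owner-relative points via a pendulum constraint, passing both endpoints to the material and generating death events.
# DroolDrops provides short-lived falling, shrinking drops; three SpriteDrips layers add small and large fading streaks.
# Skeletal-mesh-sampling variants are disabled, so active emitters use fixed offsets instead of bone sampling.
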NiagaraSystem /BRCosmetics/Effects/Fort_Effects/Emotes/Relaxed/NS_Emote_Relaxed_FaceSweat
dependencies {
 usedDataInterfaces [&NiagaraDataInterfaceCurve, &NiagaraDataInterfaceColorCurve, &NiagaraDataInterfaceVector2DCurve]
 usedDynamicInputs []
 usedModules [&SystemState]
}
emitters {
 DroolDrops {
  emitterUpdateScript [&EmitterState()]
  particleSpawnScript [
   /Niagara/Modules/Spawn/Initialization/InitializeParticle(
    Color=<rgba 1 1 1 3>
    Lifetime=&RandomRangeFloat(Minimum=0.25, Maximum=0.5)
    Mesh Scale=<5 0.5 0.5>
    Module.Write Color=true
    Module.Write Lifetime=true
    Module.Write Mass=true
    Module.Write Material Random=true
    Module.Write Position=false
    Module.Write Scale=true
    Module.Write SpriteRotation=false
    Module.Write SpriteSize=true
    Sprite Size=&Vector2DFromFloat(Value=5)
   )
   &AddVelocity(Coordinate Space=::World, Velocity=<0 0 -50>)
  ]
  particleUpdateScript [
   &ParticleState()
   &DynamicMaterialParameters(
    Index 0 Param 1=&RandomRangeFloat(Minimum=0, Maximum=1)
    Index 0 Param 2=1
    Module.Param0WriteEnabled=true
    Module.Param1WriteEnabled=true
    Module.Param2WriteEnabled=true
    Module.Param3WriteEnabled=true
   )
   &ScaleMeshSize(Scale Factor=&Multiply_VectorByFloat(
     Vector=&MakeVector(
      X=&Lerp_Float(A=0.25, B=1, Alpha=$Particles.NormalizedAge)
      Y=&Lerp_Float(A=1, B=0.5, Alpha=$Particles.NormalizedAge)
      Z=&Lerp_Float(A=1, B=0.5, Alpha=$Particles.NormalizedAge)
     )
     Float=&Lerp_Float(A=1, B=0, Alpha=$Particles.NormalizedAge)
    ))
   &GravityForce()
   &SolveForcesAndVelocity(Clamp Velocity=false, Limit Acceleration=false)
   &CameraOffset(Camera Offset Amount=15)
  ]
  rendererClasses [&NiagaraSpriteRendererProperties, &NiagaraMeshRendererProperties]
  renderers [
   {rendererClass=&NiagaraSpriteRendererProperties}
   {rendererClass=&NiagaraMeshRendererProperties, rendererIndex=1}
  ]
 }
 DroolDrops001 {
  bEnabled=false
  emitterUpdateScript [&EmitterState()]
  particleSpawnScript [
   &InitializeParticle(
    Lifetime=&RandomRangeFloat(Minimum=0.25, Maximum=0.5)
    Mesh Scale=<5 0.5 0.5>
    Module.Write Color=true
    Module.Write Lifetime=true
    Module.Write Mass=true
    Module.Write Material Random=true
    Module.Write Position=false
    Module.Write Scale=true
    Module.Write SpriteRotation=false
    Module.Write SpriteSize=true
    Sprite Size=&Vector2DFromFloat(Value=5)
   )
   &AddVelocity(Velocity=<0 0 -50>)
  ]
  particleUpdateScript [
   &ParticleState()
   &DynamicMaterialParameters(
    Index 0 Param 1=&RandomRangeFloat(Minimum=0, Maximum=1)
    Index 0 Param 2=1
    Module.Param0WriteEnabled=true
    Module.Param1WriteEnabled=true
    Module.Param2WriteEnabled=true
    Module.Param3WriteEnabled=true
   )
   &ScaleMeshSize(Scale Factor=&Multiply_VectorByFloat(
     Vector=&MakeVector(
      X=&Lerp_Float(A=0.25, B=1, Alpha=$Particles.NormalizedAge)
      Y=&Lerp_Float(A=1, B=0.5, Alpha=$Particles.NormalizedAge)
      Z=&Lerp_Float(A=1, B=0.5, Alpha=$Particles.NormalizedAge)
     )
     Float=&Lerp_Float(A=1, B=0, Alpha=$Particles.NormalizedAge)
    ))
   &GravityForce()
   &SolveForcesAndVelocity(Clamp Velocity=false, Limit Acceleration=false)
  ]
  rendererClasses [&NiagaraSpriteRendererProperties, &NiagaraMeshRendererProperties]
  renderers [
   {rendererClass=&NiagaraSpriteRendererProperties}
   {rendererClass=&NiagaraMeshRendererProperties, rendererIndex=1}
  ]
 }
 DroolPendulum {
  emitterUpdateScript [
   &EmitterState()
   &SpawnRate(
    Spawn Probability=0.1
    SpawnRate=&RandomRangeFloat(Minimum=1, Maximum=6, Evaluation Type=::NewEnumerator1)
    Use Spawn Probability=true
   ) {enabled=false}
   &SpawnRate(SpawnRate=7, Use Spawn Probability=false) {module=SpawnRate001}
  ]
  particleSpawnScript [
   /Niagara/Modules/Spawn/Initialization/V2/InitializeParticle(
    Color=<rgba 1 1 1 0.6>
    Lifetime=0.3
    Mesh Scale=<0.1 0.1 0.018>
    Mesh Scale Mode=::NewEnumerator4
    Sprite Size Mode=::NewEnumerator3
    Uniform Sprite Size=4
    UsePositionOffset=false
    Write Material Random=true
   ) {module=InitializeParticle001}
   &SphereLocation(
    MaskSpawnGroup=false
    Override Seed=false
    Sphere Coordinate Space=::World
    Sphere Origin:NiagaraPosition=<0 0 0>
    Sphere Radius=5
    Surface Only=false
    UseNonUniformScale=false
    UseOffset=false
    UseSphereOrientation=false
   )
   /SetVariables(
    Particles.BoneOffset:Vector3f=&Lerp_Vector(A=<8 0 -4>, B=<8 0 4>, Alpha=&UniformAOrBFloat(A=0))
    Particles.SpherePosition:NiagaraPosition=$Particles.Position
   ) {isSetParameters=true}
   /SetVariables(
    Particles.BoneOffset:Vector3f=&RandomRangeVector(Minimum=<12 4.5 -5>, Maximum=<12 4.5 5>)
    Particles.SpherePosition:NiagaraPosition=$Particles.Position
   ) {enabled=false, isSetParameters=true, module=SetVariables_#1}
   /SetVariables(
    Particles.BoneOffset:Vector3f=&Lerp_Vector(
     A=&RandomRangeVector(Minimum=<12 4.5 -5>, Maximum=<12 4.5 5>)
     B=&RandomRangeVector(Minimum=<-5 4.5 -5>, Maximum=<-5 4.5 5>)
     Alpha=&UniformAOrBFloat(A=0)
    )
    Particles.SpherePosition:NiagaraPosition=$Particles.Position
   ) {enabled=false, isSetParameters=true}
   /SetVariables(Particles.Position=&Add_Vector(
     A=/Niagara/DynamicInputs/Helpers/SimulationPosition()
     B=&RotateVectorByQuaternion(
      Vector To Rotate=&Add_Vector(A=$Particles.SpherePosition, B=$Particles.BoneOffset)
      Quaternion=$Engine.Owner.Rotation
     )
    )) {isSetParameters=true}
  ]
  particleUpdateScript [
   &ParticleState()
   /SetVariables(Particles.MeshPosition:Vector3f=&Add_Vector(
     A=&SimulationPosition()
     B=&RotateVectorByQuaternion(
      Vector To Rotate=&Add_Vector(A=$Particles.SpherePosition, B=$Particles.BoneOffset)
      Quaternion=$Engine.Owner.Rotation
     )
    )) {isSetParameters=true, module=SetVariables_#2}
   /SetVariables(
    Particles.AlphaScale:NiagaraFloat=&FloatFromCurve(FloatCurve=DataInterface() {propertyValues {Curve {keys [
         {arriveTangent=-0.0275, interpMode=RCIM_Cubic, leaveTangent=-0.0275, tangentMode=RCTM_User, time=0.95, value=1}
         {arriveTangent=-34.2468, interpMode=RCIM_Cubic, leaveTangent=-34.2466, tangentMode=RCTM_User, time=1, value=0}
        ]}}})
    Particles.XYScale:NiagaraFloat=&FloatFromCurve(
     FloatCurve=DataInterface() {propertyValues {Curve {keys [
         {arriveTangent=6.0986, interpMode=RCIM_Cubic, leaveTangent=6.0986, tangentMode=RCTM_User, time=0, value=0}
         {arriveTangent=-1.5439, interpMode=RCIM_Cubic, leaveTangent=-1.5439, tangentMode=RCTM_User, time=0.25, value=1}
         {arriveTangent=-0.3409, interpMode=RCIM_Cubic, leaveTangent=-0.3409, tangentMode=RCTM_User, time=0.6, value=0.6082}
         {arriveTangent=-2.6194, interpMode=RCIM_Cubic, leaveTangent=-2.6194, tangentMode=RCTM_User, time=0.9823, value=0.214}
         <at 1 0.05>
        ]}}}
     Scale Curve=&RandomRangeFloat(Minimum=0.4, Maximum=1)
    )
    Particles.ZScale:NiagaraFloat=&FloatFromCurve(
     FloatCurve=DataInterface() {propertyValues {Curve {keys [
         {arriveTangent=3.0916, interpMode=RCIM_Cubic, leaveTangent=3.0916, tangentMode=RCTM_User, time=0, value=0.1}
         {arriveTangent=0.4235, interpMode=RCIM_Cubic, leaveTangent=0.4235, tangentMode=RCTM_User, time=0.6784, value=0.9948}
         {arriveTangent=0.1347, interpMode=RCIM_Cubic, leaveTangent=0.1347, tangentMode=RCTM_User, time=0.9645, value=1.2125}
         <at 0.9984 1.5>
        ]}}}
     Scale Curve=&RandomRangeFloat(Minimum=3, Maximum=5)
    )
   ) {isSetParameters=true, module=SetVariables_#3}
   &PendulumSetup(
    Module.Enable Spring Driver=false
    Pendulum Length:NiagaraFloat=$Particles.ZScale
    Pendulum Pivot:Vector3f=$Particles.MeshPosition
    Rigid Pendulum=false
   )
   &Drag(Drag=-10, Rotational Drag=-10, Use Linear Drag=true, Use Rotational Drag=true)
   &AddVelocity(Coordinate Space=::World, Velocity=<0 0 -40>)
   &InheritVelocity(Module.Limit Speed=true, Velocity Scale=&Lerp_Vector(A=<-1 -1 -1>, B=<0 0 0>, Alpha=$Particles.NormalizedAge))
   &SolveForcesAndVelocity(Clamp Velocity=false, Limit Acceleration=false)
   &PendulumConstraint()
   &CalculateAccurateVelocity()
   /SetVariables(
    Particles.MeshFacing:Vector3f=&NormalizeVector(Vector To Normalize=&Subtract_Vector(A=$Particles.MeshPosition, B=Linked() {linkedVariable {name=Particles.Position, type=NiagaraPosition}}))
    Particles.RestDirection:Vector3f=/Niagara/DynamicInputs/Helpers/SafeNormalizeVector(Vector To Normalize:Vector3f=$Engine.Owner.Velocity, Fallback Vector:Vector3f=<0 0 0>)
   ) {enabled=false, isSetParameters=true, module=SetVariables_#4}
   /SetVariables(
    Particles.MeshFacing:Vector3f=&NormalizeVector(Vector To Normalize=&Subtract_Vector(A=$Particles.MeshPosition, B=Linked() {linkedVariable {name=Particles.Position, type=NiagaraPosition}}))
    Particles.RestDirection:Vector3f=<0 0 -1>
   ) {isSetParameters=true, module=SetVariables_#5}
   /Niagara/Modules/Update/Orientation/V2/OrientMeshToVector(
    Look At Direction:Vector3f=&Multiply_VectorByFloat(Vector=$Particles.MeshFacing, Float=1)
    Override Quat=false
    Reference Axis Vector:Vector3f=<0 0 1>
    UseOptionalSideAxis=false
    UseOptionalUpAxis=false
   )
   &DynamicMaterialParameters(
    Index 2 Param 1=HlslExpression() {hlslExpression=Particles.MeshPosition.x}
    Index 2 Param 2=HlslExpression() {hlslExpression=Particles.MeshPosition.y}
    Index 2 Param 3=HlslExpression() {hlslExpression=Particles.MeshPosition.z}
    Index 2 Param 4=&RandomRangeFloat(Minimum=2, Maximum=5)
    Index 3 Param 1=&MakeFloatFromVector(Vector=Linked() {linkedVariable {name=Particles.Position, type=NiagaraPosition}}, Channel=::NewEnumerator0)
    Index 3 Param 2=&MakeFloatFromVector(Vector=Linked() {linkedVariable {name=Particles.Position, type=NiagaraPosition}}, Channel=::NewEnumerator1)
    Index 3 Param 3=&MakeFloatFromVector(Vector=Linked() {linkedVariable {name=Particles.Position, type=NiagaraPosition}}, Channel=::NewEnumerator2)
    Index 3 Param 4=0.025
    Module.Param0WriteEnabled2=true
    Module.Param0WriteEnabled3=true
    Module.Param1WriteEnabled2=true
    Module.Param1WriteEnabled3=true
    Module.Param2WriteEnabled2=true
    Module.Param2WriteEnabled3=true
    Module.Param3WriteEnabled2=true
    Module.Param3WriteEnabled3=true
    Write Parameter Index 0=false
    Write Parameter Index 2=true
    Write Parameter Index 3=true
   )
   &DynamicMaterialParameters(
    Index 2 Param 1=HlslExpression() {hlslExpression=Particles.MeshPosition.x}
    Index 2 Param 2=HlslExpression() {hlslExpression=Particles.MeshPosition.y}
    Index 2 Param 3=HlslExpression() {hlslExpression=Particles.MeshPosition.z}
    Index 2 Param 4=&RandomRangeFloat(Minimum=2, Maximum=5)
    Index 3 Param 1=&MakeFloatFromVector(Vector=Linked() {linkedVariable {name=Particles.Position, type=NiagaraPosition}}, Channel=::NewEnumerator0)
    Index 3 Param 2=&MakeFloatFromVector(Vector=Linked() {linkedVariable {name=Particles.Position, type=NiagaraPosition}}, Channel=::NewEnumerator1)
    Index 3 Param 3=&MakeFloatFromVector(Vector=Linked() {linkedVariable {name=Particles.Position, type=NiagaraPosition}}, Channel=::NewEnumerator2)
    Index 3 Param 4=0.025
    Module.Param0WriteEnabled2=true
    Module.Param0WriteEnabled3=true
    Module.Param1WriteEnabled2=true
    Module.Param1WriteEnabled3=true
    Module.Param2WriteEnabled2=true
    Module.Param2WriteEnabled3=true
    Module.Param3WriteEnabled2=true
    Module.Param3WriteEnabled3=true
    Write Parameter Index 0=false
    Write Parameter Index 2=true
    Write Parameter Index 3=true
   ) {enabled=false, module=DynamicMaterialParameters001}
   &ScaleMeshSize(Scale Factor=&MakeVector(
     X=$Particles.XYScale
     Y=$Particles.XYScale
     Z=&VectorLength(Vector=&Subtract_Vector(A=$Particles.MeshPosition, B=Linked() {linkedVariable {name=Particles.Position, type=NiagaraPosition}}))
    ))
   &Color(Color=&ColorFromCurve(
     ColorCurve=DataInterface() {propertyValues {
       AlphaCurve {keys [<at 0 20 RCIM_Cubic>]}
       BlueCurve {keys [<at 0 0 RCIM_Cubic>,<at 0.5257 0>,<at 1 5 RCIM_Cubic>]}
       GreenCurve {keys [<at 0 0 RCIM_Cubic>,<at 0.5257 5>,<at 1 0 RCIM_Cubic>]}
       RedCurve {keys [<at 0 5 RCIM_Cubic>,<at 0.5257 0>,<at 1 0 RCIM_Cubic>]}
      }}
     CurveIndex=&Divide_Float(A=&MakeFloatFromInt(Int=$Particles.PickBone), B=366)
    )) {enabled=false}
   &ScaleColor(Scale Alpha=$Particles.AlphaScale, ScaleA=true, ScaleRGB=false)
   &GenerateDeathEvent(Event Condition:NiagaraBool=&SetBoolByFloatComparison(A=$Particles.NormalizedAge, B=0.1))
   &CameraOffset(Camera Offset Amount=15)
  ]
  rendererClasses [&NiagaraSpriteRendererProperties, &NiagaraMeshRendererProperties]
  renderers [
   {rendererClass=&NiagaraSpriteRendererProperties}
   {rendererClass=&NiagaraMeshRendererProperties, rendererIndex=1}
  ]
 }
 DroolPendulum001 {
  bEnabled=false
  emitterUpdateScript [
   &EmitterState()
   &SpawnRate(
    Spawn Probability=0.1
    SpawnRate=&RandomRangeFloat(Minimum=1, Maximum=6, Evaluation Type=::NewEnumerator1)
    Use Spawn Probability=true
   ) {enabled=false}
   &SpawnRate(SpawnRate=75, Use Spawn Probability=false) {module=SpawnRate001}
  ]
  particleSpawnScript [
   &SphereLocation(
    MaskSpawnGroup=false
    Override Seed=false
    Sphere Radius=6.5
    Surface Only=false
    UseNonUniformScale=false
    UseOffset=false
    UseSphereOrientation=false
   )
   /SetVariables(
    Particles.BoneOffset:Vector3f=&RandomRangeVector(Minimum=<10 5 0>, Maximum=<2 5 0>)
    Particles.SpherePosition:NiagaraPosition=$Particles.Position
   ) {isSetParameters=true, module=SetVariables_#1}
   &SkeletalMeshLocation(
    Bone Sampling Mode=::NewEnumerator4
    Inherit Velocity=false
    MaskSpawnGroup=false
    Sampled Position Offset=&Add_Vector(A=$Particles.SpherePosition, B=$Particles.BoneOffset)
    Sampled Position Offset Coordinate Space:ENiagaraMeshTransforms=::NewEnumerator3
    Skeletal Mesh=$User.SkeletalMesh
   ) {module=SkeletalMeshLocation002}
   /Niagara/Modules/Spawn/Initialization/InitializeParticle(
    Lifetime=&UniformRangedFloat(Minimum=0.25, Maximum=0.25)
    Mesh Scale=<0.1 0.1 0.018>
    Module.Write Color=true
    Module.Write Lifetime=true
    Module.Write Mass=true
    Module.Write Material Random=true
    Module.Write Position=false
    Module.Write Scale=true
    Module.Write SpriteRotation=false
    Module.Write SpriteSize=true
    Sprite Size=&Vector2DFromFloat(Value=10)
   )
  ]
  particleUpdateScript [
   &ParticleState()
   /SetVariables(Particles.OrigPos:NiagaraPosition=$Particles.Position) {isSetParameters=true}
   &SkeletalMeshLocation(
    Bone Sampling Mode=::NewEnumerator4
    Inherit Velocity=false
    MaskSpawnGroup=false
    Sampled Position Offset=&Add_Vector(A=$Particles.SpherePosition, B=$Particles.BoneOffset)
    Sampled Position Offset Coordinate Space:ENiagaraMeshTransforms=::NewEnumerator3
    Skeletal Mesh=$User.SkeletalMesh
   )
   /SetVariables(Particles.MeshPosition:NiagaraPosition=$Particles.Position) {isSetParameters=true, module=SetVariables_#2}
   /SetVariables(Particles.Position:Vector3f=$Particles.OrigPos) {isSetParameters=true}
   /SetVariables(
    Particles.AlphaScale:NiagaraFloat=&FloatFromCurve(FloatCurve=DataInterface() {propertyValues {Curve {keys [
         {arriveTangent=-0.0275, interpMode=RCIM_Cubic, leaveTangent=-0.0275, tangentMode=RCTM_User, time=0.95, value=1}
         {arriveTangent=-34.2468, interpMode=RCIM_Cubic, leaveTangent=-34.2466, tangentMode=RCTM_User, time=1, value=0}
        ]}}})
    Particles.XYScale:NiagaraFloat=&FloatFromCurve(
     FloatCurve=DataInterface() {propertyValues {Curve {keys [
         {arriveTangent=6.0986, interpMode=RCIM_Cubic, leaveTangent=6.0986, tangentMode=RCTM_User, time=0, value=0}
         {arriveTangent=-1.5439, interpMode=RCIM_Cubic, leaveTangent=-1.5439, tangentMode=RCTM_User, time=0.25, value=1}
         {arriveTangent=-0.3409, interpMode=RCIM_Cubic, leaveTangent=-0.3409, tangentMode=RCTM_User, time=0.6, value=0.6082}
         {arriveTangent=-2.6194, interpMode=RCIM_Cubic, leaveTangent=-2.6194, tangentMode=RCTM_User, time=0.9823, value=0.214}
         <at 1 0.05>
        ]}}}
     Scale Curve=&RandomRangeFloat(Minimum=0.4, Maximum=1)
    )
    Particles.ZScale:NiagaraFloat=&FloatFromCurve(
     FloatCurve=DataInterface() {propertyValues {Curve {keys [
         {arriveTangent=3.0916, interpMode=RCIM_Cubic, leaveTangent=3.0916, tangentMode=RCTM_User, time=0, value=0.1}
         {arriveTangent=0.4235, interpMode=RCIM_Cubic, leaveTangent=0.4235, tangentMode=RCTM_User, time=0.6784, value=0.9948}
         {arriveTangent=0.1347, interpMode=RCIM_Cubic, leaveTangent=0.1347, tangentMode=RCTM_User, time=0.9645, value=1.2125}
         <at 0.9984 1.5>
        ]}}}
     Scale Curve=&RandomRangeFloat(Minimum=3, Maximum=10)
    )
   ) {isSetParameters=true, module=SetVariables_#3}
   &PendulumSetup(
    Module.Enable Spring Driver=false
    Pendulum Length:NiagaraFloat=$Particles.ZScale
    Pendulum Pivot:Vector3f=$Particles.MeshPosition
    Rigid Pendulum=false
   )
   &Drag(Drag=-10, Rotational Drag=-10, Use Linear Drag=true, Use Rotational Drag=true)
   &AddVelocity(Velocity=<0 0 -40>)
   &InheritVelocity(Module.Limit Speed=true, Velocity Scale=&Lerp_Vector(A=<-1 -1 -1>, B=<0 0 0>, Alpha=$Particles.NormalizedAge))
   &SolveForcesAndVelocity(Clamp Velocity=false, Limit Acceleration=false)
   &PendulumConstraint()
   &CalculateAccurateVelocity()
   /SetVariables(
    Particles.MeshFacing:Vector3f=&NormalizeVector(Vector To Normalize=&Subtract_Vector(A=$Particles.MeshPosition, B=Linked() {linkedVariable {name=Particles.Position, type=NiagaraPosition}}))
    Particles.RestDirection:Vector3f=&SafeNormalizeVector(Vector To Normalize:Vector3f=$Engine.Owner.Velocity, Fallback Vector:Vector3f=<0 0 0>)
   ) {enabled=false, isSetParameters=true, module=SetVariables_#4}
   /SetVariables(
    Particles.MeshFacing:Vector3f=&NormalizeVector(Vector To Normalize=&Subtract_Vector(A=$Particles.MeshPosition, B=Linked() {linkedVariable {name=Particles.Position, type=NiagaraPosition}}))
    Particles.RestDirection:Vector3f=<0 0 -1>
   ) {isSetParameters=true, module=SetVariables_#5}
   &OrientMeshToVector(
    Look At Direction:Vector3f=&Multiply_VectorByFloat(Vector=$Particles.MeshFacing, Float=1)
    Override Quat=false
    Reference Axis Vector:Vector3f=<0 0 1>
    UseOptionalSideAxis=false
    UseOptionalUpAxis=false
   )
   &DynamicMaterialParameters(
    Index 2 Param 1=HlslExpression() {hlslExpression=Particles.MeshPosition.x}
    Index 2 Param 2=HlslExpression() {hlslExpression=Particles.MeshPosition.y}
    Index 2 Param 3=HlslExpression() {hlslExpression=Particles.MeshPosition.z}
    Index 2 Param 4=&RandomRangeFloat(Minimum=2, Maximum=5)
    Index 3 Param 1=&MakeFloatFromVector(Vector=Linked() {linkedVariable {name=Particles.Position, type=NiagaraPosition}}, Channel=::NewEnumerator0)
    Index 3 Param 2=&MakeFloatFromVector(Vector=Linked() {linkedVariable {name=Particles.Position, type=NiagaraPosition}}, Channel=::NewEnumerator1)
    Index 3 Param 3=&MakeFloatFromVector(Vector=Linked() {linkedVariable {name=Particles.Position, type=NiagaraPosition}}, Channel=::NewEnumerator2)
    Index 3 Param 4=0.025
    Module.Param0WriteEnabled2=true
    Module.Param0WriteEnabled3=true
    Module.Param1WriteEnabled2=true
    Module.Param1WriteEnabled3=true
    Module.Param2WriteEnabled2=true
    Module.Param2WriteEnabled3=true
    Module.Param3WriteEnabled2=true
    Module.Param3WriteEnabled3=true
    Write Parameter Index 0=false
    Write Parameter Index 2=true
    Write Parameter Index 3=true
   )
   &ScaleMeshSize(Scale Factor=&MakeVector(
     X=$Particles.XYScale
     Y=$Particles.XYScale
     Z=&VectorLength(Vector=&Subtract_Vector(A=$Particles.MeshPosition, B=Linked() {linkedVariable {name=Particles.Position, type=NiagaraPosition}}))
    ))
   &Color(Color=&ColorFromCurve(
     ColorCurve=DataInterface() {propertyValues {
       AlphaCurve {keys [<at 0 20 RCIM_Cubic>]}
       BlueCurve {keys [<at 0 0 RCIM_Cubic>,<at 0.5257 0>,<at 1 5 RCIM_Cubic>]}
       GreenCurve {keys [<at 0 0 RCIM_Cubic>,<at 0.5257 5>,<at 1 0 RCIM_Cubic>]}
       RedCurve {keys [<at 0 5 RCIM_Cubic>,<at 0.5257 0>,<at 1 0 RCIM_Cubic>]}
      }}
     CurveIndex=&Divide_Float(A=&MakeFloatFromInt(Int=$Particles.PickBone), B=366)
    )) {enabled=false}
   &ScaleColor(Scale Alpha=$Particles.AlphaScale, ScaleA=true, ScaleRGB=false)
   &GenerateDeathEvent(Event Condition:NiagaraBool=&SetBoolByFloatComparison(A=$Particles.NormalizedAge, B=0.1))
  ]
  rendererClasses [&NiagaraSpriteRendererProperties, &NiagaraMeshRendererProperties]
  renderers [
   {rendererClass=&NiagaraSpriteRendererProperties}
   {rendererClass=&NiagaraMeshRendererProperties, rendererIndex=1}
  ]
 }
 Empty {
  bEnabled=false
  emitterUpdateScript [
   &EmitterState()
   &SpawnBurst_Instantaneous(Spawn Count=5, Use Spawn Probability=false)
  ]
  particleSpawnScript [/Niagara/Modules/Spawn/Initialization/V2/InitializeParticle(
    Color Mode=::NewEnumerator0
    Lifetime=2
    Mesh Scale Mode=::NewEnumerator3
    Mesh Uniform Scale=5
    Sprite Size Mode=::NewEnumerator3
    UsePositionOffset=false
    Write Material Random=true
   )]
  particleUpdateScript [
   &ParticleState()
   &Color(Color=&ColorFromCurve(ColorCurve=DataInterface() {propertyValues {
       AlphaCurve {keys [<at 0 20 RCIM_Cubic>]}
       BlueCurve {keys [<at 0 0 RCIM_Cubic>,<at 0.5257 0>,<at 1 5 RCIM_Cubic>]}
       GreenCurve {keys [<at 0 0 RCIM_Cubic>,<at 0.5257 5>,<at 1 0 RCIM_Cubic>]}
       RedCurve {keys [<at 0 5 RCIM_Cubic>,<at 0.5257 0>,<at 1 0 RCIM_Cubic>]}
      }})) {enabled=false}
   &SkeletalMeshLocation(
    Apply Sampled Orientation to Mesh Particles=true
    Bone Sampling Mode=::NewEnumerator4
    Inherit Velocity=false
    MaskSpawnGroup=false
    Sampled Position Offset=<0 6.75 0>
    Sampled Position Offset Coordinate Space:ENiagaraMeshTransforms=::NewEnumerator3
    Skeletal Mesh=$User.SkeletalMesh
   ) {module=SkeletalMeshLocation002}
  ]
  rendererClasses [&NiagaraSpriteRendererProperties, &NiagaraMeshRendererProperties]
  renderers [
   {rendererClass=&NiagaraSpriteRendererProperties}
   {rendererClass=&NiagaraMeshRendererProperties, rendererIndex=1}
  ]
 }
 SpriteDrips {
  emitterUpdateScript [
   &EmitterState()
   &SpawnRate(SpawnRate=50, Use Spawn Probability=false)
  ]
  particleSpawnScript [
   &SphereLocation(
    MaskSpawnGroup=false
    Offset=<-1 0 0>
    Override Seed=false
    Sphere Radius=10
    Surface Only=false
    UseNonUniformScale=false
    UseOffset=false
    UseSphereOrientation=false
   )
   /SetVariables(
    Particles.BoneOffset:Vector3f=&RandomRangeVector(Minimum=<10 5 0>, Maximum=<4 5 0>)
    Particles.SpherePosition:NiagaraPosition=$Particles.Position
   ) {isSetParameters=true, module=SetVariables_#6}
   &InitializeParticle(
    Color=<rgba 1 1 1 0.7>
    Lifetime Max=0.25
    Lifetime Min=0.1
    Lifetime Mode=::NewEnumerator1
    Sprite Size Max=<0.8 0.8>
    Sprite Size Min=<0.32 0.6>
    Sprite Size Mode=::NewEnumerator2
    UsePositionOffset=false
    Write Material Random=true
   )
   /SetVariables(
    Particles.Position=&Add_Vector(A=$Particles.SpherePosition, B=$Particles.BoneOffset)
    Particles.SpriteAlignment=&TransformVector(Vector=<0 0 -1>, Source Space=::World, Destination Space=::Local)
   ) {isSetParameters=true, module=SetVariables_#7}
  ]
  particleUpdateScript [
   &ParticleState()
   &SkeletalMeshLocation(
    Bone Sampling Mode=::NewEnumerator4
    Inherit Velocity=false
    MaskSpawnGroup=false
    Sampled Position Offset=&Add_Vector(
     A=&Add_Vector(
      A=&Multiply_VectorByFloat(
       Vector=$Particles.SpherePosition
       Float=&Add_Float(A=1, B=&Multiply_Float(A=$Particles.NormalizedAge, B=-0.5))
      )
      B=$Particles.BoneOffset
     )
     B=&Multiply_VectorByFloat(Vector=&RandomRangeVector(Minimum=<-6 0 0>, Maximum=<-10 0 0>), Float=$Particles.NormalizedAge)
    )
    Sampled Position Offset Coordinate Space:ENiagaraMeshTransforms=::NewEnumerator3
    Skeletal Mesh=$User.SkeletalMesh
   ) {enabled=false}
   /SetVariables(Particles.Position=&Add_Vector(
     A=&Add_Vector(
      A=&Multiply_VectorByFloat(
       Vector=$Particles.SpherePosition
       Float=&Add_Float(A=1, B=&Multiply_Float(A=$Particles.NormalizedAge, B=-0.2))
      )
      B=$Particles.BoneOffset
     )
     B=&Multiply_VectorByFloat(Vector=&RandomRangeVector(Minimum=<-6 0 0>, Maximum=<-10 0 0>), Float=$Particles.NormalizedAge)
    )) {isSetParameters=true, module=SetVariables_#8}
   /SetVariables(Particles.Position=&Add_Vector(
     A=&Add_Vector(
      A=&Multiply_VectorByFloat(
       Vector=$Particles.SpherePosition
       Float=&Add_Float(A=1, B=&Multiply_Float(A=$Particles.NormalizedAge, B=-0.5))
      )
      B=$Particles.BoneOffset
     )
     B=&Multiply_VectorByFloat(
      Vector=&TransformVector(
       Vector=&RandomRangeVector(Minimum=<0 0 -6>, Maximum=<0 0 -10>)
       Source Space=::World
       Destination Space=::Local
      )
      Float=$Particles.NormalizedAge
     )
    )) {enabled=false, isSetParameters=true}
   &ScaleSpriteSize(Scale Factor=&Vector2DFromCurve(
     Vector2Curve=DataInterface() {propertyValues {
       XCurve {keys [<at 0.25 1 RCIM_Cubic>, {interpMode=RCIM_Cubic, leaveTangent=-2.9373, time=1, value=0.3}]}
       YCurve {keys [
         {arriveTangent=-1.4216, interpMode=RCIM_Cubic, time=0.0067, value=0.3563}
         {arriveTangent=2.9367, interpMode=RCIM_Cubic, leaveTangent=2.9367, tangentMode=RCTM_User, time=0.2953, value=1.3705}
         {interpMode=RCIM_Cubic, leaveTangent=7.4188, time=1, value=2}
        ]}
      }}
     ScaleCurve=<1 1>
    ))
   &DynamicMaterialParameters(
    Index 0 Param 1=0.8
    Index 0 Param 2=&FloatFromCurve(FloatCurve=DataInterface() {propertyValues {Curve {keys [{arriveTangent=-1.8884, interpMode=RCIM_Cubic, leaveTangent=-1.8884, tangentMode=RCTM_User, time=0, value=1}, <at 1 0.26>]}}})
    Index 0 Param 3=1
    Index 0 Param 4=1
    Module.Param0WriteEnabled=true
    Module.Param1WriteEnabled=true
    Module.Param2WriteEnabled=false
    Module.Param3WriteEnabled=false
   )
   &CameraOffset(Camera Offset Amount=&RandomRangeFloat(Minimum=20, Maximum=20))
   &ScaleColor(
    Scale Alpha=&FloatFromCurve(FloatCurve=DataInterface() {propertyValues {Curve {keys [<at 0 1>, {arriveTangent=-1, interpMode=RCIM_Cubic, leaveTangent=-1, time=0.8, value=1}, <at 1 0>]}}})
    ScaleA=true
    ScaleRGB=true
   )
  ]
  rendererClasses [&NiagaraSpriteRendererProperties]
  renderers [{rendererClass=&NiagaraSpriteRendererProperties}]
 }
 SpriteDrips001 {
  bEnabled=false
  emitterUpdateScript [
   &EmitterState()
   &SpawnRate(SpawnRate=100, Use Spawn Probability=false)
  ]
  particleSpawnScript [
   &SphereLocation(
    MaskSpawnGroup=false
    Override Seed=false
    Sphere Radius=6.5
    Surface Only=false
    UseNonUniformScale=false
    UseOffset=false
    UseSphereOrientation=false
   )
   /SetVariables(
    Particles.BoneOffset:Vector3f=&RandomRangeVector(Minimum=<10 5 0>, Maximum=<4 5 0>)
    Particles.SpherePosition:NiagaraPosition=$Particles.Position
   ) {isSetParameters=true, module=SetVariables_#6}
   &InitializeParticle(
    Color=<rgba 1 1 1 0.25>
    Lifetime Max=0.25
    Lifetime Min=0.1
    Lifetime Mode=::NewEnumerator1
    Sprite Size Max=<2 3>
    Sprite Size Min=<1 2>
    Sprite Size Mode=::NewEnumerator2
    UsePositionOffset=false
    Write Material Random=true
   )
   &SkeletalMeshLocation(
    Bone Sampling Mode=::NewEnumerator4
    Inherit Velocity=false
    MaskSpawnGroup=false
    Sampled Position Offset=&Add_Vector(A=$Particles.SpherePosition, B=$Particles.BoneOffset)
    Sampled Position Offset Coordinate Space:ENiagaraMeshTransforms=::NewEnumerator3
    Skeletal Mesh=$User.SkeletalMesh
   ) {module=SkeletalMeshLocation002}
   /SetVariables(Particles.SpriteAlignment=<0 0 -1>) {isSetParameters=true}
  ]
  particleUpdateScript [
   &ParticleState()
   &SkeletalMeshLocation(
    Bone Sampling Mode=::NewEnumerator4
    Inherit Velocity=false
    MaskSpawnGroup=false
    Sampled Position Offset=&Add_Vector(
     A=&Add_Vector(
      A=&Multiply_VectorByFloat(
       Vector=$Particles.SpherePosition
       Float=&Add_Float(A=1, B=&Multiply_Float(A=$Particles.NormalizedAge, B=-0.5))
      )
      B=$Particles.BoneOffset
     )
     B=&Multiply_VectorByFloat(Vector=&RandomRangeVector(Minimum=<-6 0 0>, Maximum=<-10 0 0>), Float=$Particles.NormalizedAge)
    )
    Sampled Position Offset Coordinate Space:ENiagaraMeshTransforms=::NewEnumerator3
    Skeletal Mesh=$User.SkeletalMesh
   )
   &ScaleSpriteSize(Scale Factor=&Vector2DFromCurve(
     Vector2Curve=DataInterface() {propertyValues {
       XCurve {keys [<at 0.25 1 RCIM_Cubic>, {interpMode=RCIM_Cubic, leaveTangent=-2.9373, time=1, value=0.3}]}
       YCurve {keys [
         {arriveTangent=-1.4216, interpMode=RCIM_Cubic, time=0.0067, value=0.3563}
         {arriveTangent=2.9367, interpMode=RCIM_Cubic, leaveTangent=2.9367, tangentMode=RCTM_User, time=0.2953, value=1.3705}
         {interpMode=RCIM_Cubic, leaveTangent=7.4188, time=1, value=2}
        ]}
      }}
     ScaleCurve=<1 1>
    ))
   &DynamicMaterialParameters(
    Index 0 Param 1=0.8
    Index 0 Param 2=&FloatFromCurve(FloatCurve=DataInterface() {propertyValues {Curve {keys [{arriveTangent=-1.8884, interpMode=RCIM_Cubic, leaveTangent=-1.8884, tangentMode=RCTM_User, time=0, value=1}, <at 1 0.26>]}}})
    Index 0 Param 3=1
    Index 0 Param 4=1
    Module.Param0WriteEnabled=true
    Module.Param1WriteEnabled=true
    Module.Param2WriteEnabled=false
    Module.Param3WriteEnabled=false
   )
   &CameraOffset(Camera Offset Amount=&RandomRangeFloat(Minimum=6, Maximum=10))
   &ScaleColor(
    Scale Alpha=&FloatFromCurve(FloatCurve=DataInterface() {propertyValues {Curve {keys [<at 0 1>, {arriveTangent=-1, interpMode=RCIM_Cubic, leaveTangent=-1, time=0.8, value=1}, <at 1 0>]}}})
    ScaleA=true
    ScaleRGB=true
   )
  ]
  rendererClasses [&NiagaraSpriteRendererProperties]
  renderers [{rendererClass=&NiagaraSpriteRendererProperties}]
 }
 SpriteDripsBig {
  emitterUpdateScript [
   &EmitterState()
   &SpawnRate(SpawnRate=37.5, Use Spawn Probability=false)
  ]
  particleSpawnScript [
   &SphereLocation(
    MaskSpawnGroup=false
    Offset=<-1 0 0>
    Override Seed=false
    Sphere Radius=10
    Surface Only=false
    UseNonUniformScale=false
    UseOffset=false
    UseSphereOrientation=false
   )
   /SetVariables(
    Particles.BoneOffset:Vector3f=&RandomRangeVector(Minimum=<10 5 0>, Maximum=<4 5 0>)
    Particles.SpherePosition:NiagaraPosition=$Particles.Position
   ) {isSetParameters=true, module=SetVariables_#6}
   &InitializeParticle(
    Color=<rgba 1 1 1 0.7>
    Lifetime Max=0.25
    Lifetime Min=0.1
    Lifetime Mode=::NewEnumerator1
    Sprite Size Max=<2.5 2.5>
    Sprite Size Min=<0.8 1.5>
    Sprite Size Mode=::NewEnumerator2
    UsePositionOffset=false
    Write Material Random=true
   )
   /SetVariables(
    Particles.Position=&Add_Vector(A=$Particles.SpherePosition, B=$Particles.BoneOffset)
    Particles.SpriteAlignment=&TransformVector(Vector=<0 0 -1>, Source Space=::World, Destination Space=::Local)
   ) {isSetParameters=true, module=SetVariables_#7}
  ]
  particleUpdateScript [
   &ParticleState()
   &SkeletalMeshLocation(
    Bone Sampling Mode=::NewEnumerator4
    Inherit Velocity=false
    MaskSpawnGroup=false
    Sampled Position Offset=&Add_Vector(
     A=&Add_Vector(
      A=&Multiply_VectorByFloat(
       Vector=$Particles.SpherePosition
       Float=&Add_Float(A=1, B=&Multiply_Float(A=$Particles.NormalizedAge, B=-0.5))
      )
      B=$Particles.BoneOffset
     )
     B=&Multiply_VectorByFloat(Vector=&RandomRangeVector(Minimum=<-6 0 0>, Maximum=<-10 0 0>), Float=$Particles.NormalizedAge)
    )
    Sampled Position Offset Coordinate Space:ENiagaraMeshTransforms=::NewEnumerator3
    Skeletal Mesh=$User.SkeletalMesh
   ) {enabled=false}
   /SetVariables(Particles.Position=&Add_Vector(
     A=&Add_Vector(
      A=&Multiply_VectorByFloat(
       Vector=$Particles.SpherePosition
       Float=&Add_Float(A=1, B=&Multiply_Float(A=$Particles.NormalizedAge, B=-0.2))
      )
      B=$Particles.BoneOffset
     )
     B=&Multiply_VectorByFloat(Vector=&RandomRangeVector(Minimum=<-6 0 0>, Maximum=<-10 0 0>), Float=$Particles.NormalizedAge)
    )) {isSetParameters=true, module=SetVariables_#8}
   &ScaleSpriteSize(Scale Factor=&Vector2DFromCurve(
     Vector2Curve=DataInterface() {propertyValues {
       XCurve {keys [<at 0.25 1 RCIM_Cubic>, {interpMode=RCIM_Cubic, leaveTangent=-2.9373, time=1, value=0.3}]}
       YCurve {keys [
         {arriveTangent=-1.4216, interpMode=RCIM_Cubic, time=0.0067, value=0.3563}
         {arriveTangent=2.9367, interpMode=RCIM_Cubic, leaveTangent=2.9367, tangentMode=RCTM_User, time=0.2953, value=1.3705}
         {interpMode=RCIM_Cubic, leaveTangent=7.4188, time=1, value=2}
        ]}
      }}
     ScaleCurve=<1 1>
    ))
   &DynamicMaterialParameters(
    Index 0 Param 1=0.8
    Index 0 Param 2=&FloatFromCurve(FloatCurve=DataInterface() {propertyValues {Curve {keys [{arriveTangent=-1.8884, interpMode=RCIM_Cubic, leaveTangent=-1.8884, tangentMode=RCTM_User, time=0, value=1}, <at 1 0.26>]}}})
    Index 0 Param 3=1
    Index 0 Param 4=1
    Module.Param0WriteEnabled=true
    Module.Param1WriteEnabled=true
    Module.Param2WriteEnabled=false
    Module.Param3WriteEnabled=false
   )
   &CameraOffset(Camera Offset Amount=&RandomRangeFloat(Minimum=20, Maximum=20))
   &ScaleColor(
    Scale Alpha=&FloatFromCurve(
     FloatCurve=DataInterface() {propertyValues {Curve {keys [<at 0 1>, {arriveTangent=-1, interpMode=RCIM_Cubic, leaveTangent=-1, time=0.8, value=1}, <at 1 0>]}}}
     Scale Curve=&RandomRangeFloat(Minimum=0.1, Maximum=0.3)
    )
    ScaleA=true
    ScaleRGB=false
   )
  ]
  rendererClasses [&NiagaraSpriteRendererProperties]
  renderers [{rendererClass=&NiagaraSpriteRendererProperties}]
 }
 SpriteDripsBig001 {
  emitterUpdateScript [
   &EmitterState()
   &SpawnRate(SpawnRate=37.5, Use Spawn Probability=false)
  ]
  particleSpawnScript [
   &SphereLocation(
    MaskSpawnGroup=false
    Offset=<-1 0 0>
    Override Seed=false
    Sphere Radius=10
    Surface Only=false
    UseNonUniformScale=false
    UseOffset=false
    UseSphereOrientation=false
   )
   /SetVariables(
    Particles.BoneOffset:Vector3f=&RandomRangeVector(Minimum=<10 5 0>, Maximum=<4 5 0>)
    Particles.SpherePosition:NiagaraPosition=$Particles.Position
   ) {isSetParameters=true, module=SetVariables_#6}
   &InitializeParticle(
    Color=<rgba 1 1 1 2>
    Lifetime Max=0.25
    Lifetime Min=0.1
    Lifetime Mode=::NewEnumerator1
    Sprite Size Max=<2.5 2.5>
    Sprite Size Min=<0.8 1.5>
    Sprite Size Mode=::NewEnumerator2
    UsePositionOffset=false
    Write Material Random=true
   )
   /SetVariables(
    Particles.Position=&Add_Vector(A=$Particles.SpherePosition, B=$Particles.BoneOffset)
    Particles.SpriteAlignment=&TransformVector(Vector=<0 0 -1>, Source Space=::World, Destination Space=::Local)
   ) {isSetParameters=true, module=SetVariables_#7}
  ]
  particleUpdateScript [
   &ParticleState()
   &SkeletalMeshLocation(
    Bone Sampling Mode=::NewEnumerator4
    Inherit Velocity=false
    MaskSpawnGroup=false
    Sampled Position Offset=&Add_Vector(
     A=&Add_Vector(
      A=&Multiply_VectorByFloat(
       Vector=$Particles.SpherePosition
       Float=&Add_Float(A=1, B=&Multiply_Float(A=$Particles.NormalizedAge, B=-0.5))
      )
      B=$Particles.BoneOffset
     )
     B=&Multiply_VectorByFloat(Vector=&RandomRangeVector(Minimum=<-6 0 0>, Maximum=<-10 0 0>), Float=$Particles.NormalizedAge)
    )
    Sampled Position Offset Coordinate Space:ENiagaraMeshTransforms=::NewEnumerator3
    Skeletal Mesh=$User.SkeletalMesh
   ) {enabled=false}
   /SetVariables(Particles.Position=&Add_Vector(
     A=&Add_Vector(
      A=&Multiply_VectorByFloat(
       Vector=$Particles.SpherePosition
       Float=&Add_Float(A=1, B=&Multiply_Float(A=$Particles.NormalizedAge, B=-0.2))
      )
      B=$Particles.BoneOffset
     )
     B=&Multiply_VectorByFloat(Vector=&RandomRangeVector(Minimum=<-6 0 0>, Maximum=<-10 0 0>), Float=$Particles.NormalizedAge)
    )) {isSetParameters=true, module=SetVariables_#8}
   &ScaleSpriteSize(Scale Factor=&Vector2DFromCurve(
     Vector2Curve=DataInterface() {propertyValues {
       XCurve {keys [<at 0.25 1 RCIM_Cubic>, {interpMode=RCIM_Cubic, leaveTangent=-2.9373, time=1, value=0.3}]}
       YCurve {keys [
         {arriveTangent=-1.4216, interpMode=RCIM_Cubic, time=0.0067, value=0.3563}
         {arriveTangent=2.9367, interpMode=RCIM_Cubic, leaveTangent=2.9367, tangentMode=RCTM_User, time=0.2953, value=1.3705}
         {interpMode=RCIM_Cubic, leaveTangent=7.4188, time=1, value=2}
        ]}
      }}
     ScaleCurve=<1 1>
    ))
   &DynamicMaterialParameters(
    Index 0 Param 1=0.8
    Index 0 Param 2=&FloatFromCurve(FloatCurve=DataInterface() {propertyValues {Curve {keys [{arriveTangent=-1.8884, interpMode=RCIM_Cubic, leaveTangent=-1.8884, tangentMode=RCTM_User, time=0, value=1}, <at 1 0.26>]}}})
    Index 0 Param 3=1
    Index 0 Param 4=1
    Module.Param0WriteEnabled=true
    Module.Param1WriteEnabled=true
    Module.Param2WriteEnabled=false
    Module.Param3WriteEnabled=false
   )
   &CameraOffset(Camera Offset Amount=&RandomRangeFloat(Minimum=20, Maximum=20))
   &ScaleColor(
    Scale Alpha=&FloatFromCurve(
     FloatCurve=DataInterface() {propertyValues {Curve {keys [<at 0 1>, {arriveTangent=-1, interpMode=RCIM_Cubic, leaveTangent=-1, time=0.8, value=1}, <at 1 0>]}}}
     Scale Curve=&RandomRangeFloat(Minimum=0.1, Maximum=0.3)
    )
    ScaleA=true
    ScaleRGB=false
   )
  ]
  rendererClasses [&NiagaraSpriteRendererProperties]
  renderers [{rendererClass=&NiagaraSpriteRendererProperties}]
 }
}
summary {
 systemName=NS_Emote_Relaxed_FaceSweat
 emitters [
  {emitterName=DroolDrops, rendererClasses [&NiagaraSpriteRendererProperties, &NiagaraMeshRendererProperties]}
  {emitterName=DroolPendulum, rendererClasses [&NiagaraSpriteRendererProperties, &NiagaraMeshRendererProperties]}
  {
   bEnabled=false
   emitterName=Empty
   rendererClasses [&NiagaraSpriteRendererProperties, &NiagaraMeshRendererProperties]
  }
  {emitterName=SpriteDrips, rendererClasses [&NiagaraSpriteRendererProperties]}
  {
   bEnabled=false
   emitterName=DroolDrops001
   rendererClasses [&NiagaraSpriteRendererProperties, &NiagaraMeshRendererProperties]
  }
  {
   bEnabled=false
   emitterName=DroolPendulum001
   rendererClasses [&NiagaraSpriteRendererProperties, &NiagaraMeshRendererProperties]
  }
  {bEnabled=false, emitterName=SpriteDrips001, rendererClasses [&NiagaraSpriteRendererProperties]}
  {emitterName=SpriteDripsBig, rendererClasses [&NiagaraSpriteRendererProperties]}
  {emitterName=SpriteDripsBig001, rendererClasses [&NiagaraSpriteRendererProperties]}
 ]
 userVariables [
  User.bIsFrontEndPreview {type=NiagaraBool}
  User.bIsSkydiving {type=NiagaraBool}
  User.ForwardDot {type=NiagaraFloat, defaultValue=0, description=""}
  User.IsSkydiving {type=NiagaraInt32}
  User.JumpTime {type=NiagaraFloat, defaultValue=0, description=""}
  User.ReactiveRate {type=NiagaraFloat, defaultValue=0, description=""}
  User.RunningSpeed {type=NiagaraFloat, defaultValue=0, description=""}
  User.SkeletalMesh {
   type=NiagaraDataInterfaceSkeletalMesh
   defaultValue {
    struct=NiagaraExt_VariableValue_DataInterface
    value {dataInterface=&self:NiagaraDataInterfaceSkeletalMesh_0}
   }
  }
  User.SpawnBurst {type=NiagaraBool}
 ]
}
