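# Niagara system for the Relaxed emote's hand burst, spraying small bluish-white flying drops.
# Two emitters, identical except for color alpha, use probabilistic rate spawning inside a small sphere, inherit velocity and fall under gravity.
# Randomly elongated particles, with sprite and mesh renderers, reshape and shrink to nothing over their short lifetimes.
# Skeletal-mesh bone placement is disabled; user variables like SpawnBurst are declared but drive no active module shown.
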NiagaraSystem /BRCosmetics/Effects/Fort_Effects/Emotes/Relaxed/NS_Emote_Relaxed_HandBurst
dependencies {usedDataInterfaces [], usedDynamicInputs [], usedModules [&SystemState]}
emitters {
 FlyingDrops {
  emitterUpdateScript [
   &EmitterState()
   &SpawnRate(Spawn Probability=0.5, SpawnRate=250, Use Spawn Probability=true)
  ]
  particleSpawnScript [
   /SetVariables(Particles.YZSize:NiagaraFloat=&RandomRangeFloat(Minimum=0.4, Maximum=0.75)) {isSetParameters=true, module=SetVariables_#1}
   &InitializeParticle(
    Color=&RandomRangeLinearColor(Minimum=<rgba 0.6042 0.7437 1 0.5>, Maximum=<rgba 1 1 1 2>, Color Channel Mode=::NewEnumerator2)
    Lifetime Max=1
    Lifetime Min=0.25
    Lifetime Mode=::NewEnumerator1
    Mesh Scale=&MakeVector(X=&RandomRangeFloat(Minimum=5, Maximum=10), Y=$Particles.YZSize, Z=$Particles.YZSize)
    Mesh Scale Mode=::NewEnumerator4
    Sprite Size Mode=::NewEnumerator3
    Uniform Sprite Size=5
    UsePositionOffset=false
    Write Material Random=true
   )
   &SphereLocation(
    MaskSpawnGroup=false
    Override Seed=false
    Sphere Radius=6.5
    Surface Only=false
    UseNonUniformScale=false
    UseOffset=false
    UseSphereOrientation=false
   )
   /SetVariables(
    Particles.BoneOffset:Vector3f=&RandomRangeVector(Minimum=<10 5 0>, Maximum=<2 5 0>)
    Particles.SpherePosition:NiagaraPosition=$Particles.Position
   ) {isSetParameters=true, module=SetVariables_#2}
   &SkeletalMeshLocation(
    Bone Sampling Mode=::NewEnumerator4
    Inherit Velocity=false
    MaskSpawnGroup=false
    Sampled Position Offset=$Particles.SpherePosition
    Sampled Position Offset Coordinate Space:ENiagaraMeshTransforms=::NewEnumerator3
    Skeletal Mesh=$User.SkeletalMesh
   ) {enabled=false, module=SkeletalMeshLocation002}
   &AddVelocity(Velocity=<0 0 -50>)
   &InheritVelocity(Module.Limit Speed=false, Velocity Scale=&VectorFromFloat(Value=1))
  ]
  particleUpdateScript [
   &ParticleState()
   &DynamicMaterialParameters(
    Index 0 Param 1=&RandomRangeFloat(Minimum=0, Maximum=1)
    Index 0 Param 2=1
    Module.Param0WriteEnabled=true
    Module.Param1WriteEnabled=true
    Module.Param2WriteEnabled=true
    Module.Param3WriteEnabled=true
   )
   &ScaleMeshSize(Scale Factor=&Multiply_VectorByFloat(
     Vector=&MakeVector(
      X=&Lerp_Float(A=0.25, B=1, Alpha=$Particles.NormalizedAge)
      Y=&Lerp_Float(A=1, B=0.5, Alpha=$Particles.NormalizedAge)
      Z=&Lerp_Float(A=1, B=0.5, Alpha=$Particles.NormalizedAge)
     )
     Float=&Lerp_Float(A=1, B=0, Alpha=$Particles.NormalizedAge)
    ))
   &GravityForce()
   &SolveForcesAndVelocity(Clamp Velocity=false, Limit Acceleration=false)
  ]
  rendererClasses [&NiagaraSpriteRendererProperties, &NiagaraMeshRendererProperties]
  renderers [
   {rendererClass=&NiagaraSpriteRendererProperties}
   {rendererClass=&NiagaraMeshRendererProperties, rendererIndex=1}
  ]
 }
 FlyingDrops001 {
  emitterUpdateScript [
   &EmitterState()
   &SpawnRate(Spawn Probability=0.5, SpawnRate=250, Use Spawn Probability=true)
  ]
  particleSpawnScript [
   /SetVariables(Particles.YZSize:NiagaraFloat=&RandomRangeFloat(Minimum=0.4, Maximum=0.75)) {isSetParameters=true, module=SetVariables_#1}
   &InitializeParticle(
    Color=&RandomRangeLinearColor(Minimum=<rgba 0.6042 0.7437 1 8>, Maximum=<rgba 1 1 1 10>, Color Channel Mode=::NewEnumerator2)
    Lifetime Max=1
    Lifetime Min=0.25
    Lifetime Mode=::NewEnumerator1
    Mesh Scale=&MakeVector(X=&RandomRangeFloat(Minimum=5, Maximum=10), Y=$Particles.YZSize, Z=$Particles.YZSize)
    Mesh Scale Mode=::NewEnumerator4
    Sprite Size Mode=::NewEnumerator3
    Uniform Sprite Size=5
    UsePositionOffset=false
    Write Material Random=true
   )
   &SphereLocation(
    MaskSpawnGroup=false
    Override Seed=false
    Sphere Radius=6.5
    Surface Only=false
    UseNonUniformScale=false
    UseOffset=false
    UseSphereOrientation=false
   )
   /SetVariables(
    Particles.BoneOffset:Vector3f=&RandomRangeVector(Minimum=<10 5 0>, Maximum=<2 5 0>)
    Particles.SpherePosition:NiagaraPosition=$Particles.Position
   ) {isSetParameters=true, module=SetVariables_#2}
   &SkeletalMeshLocation(
    Bone Sampling Mode=::NewEnumerator4
    Inherit Velocity=false
    MaskSpawnGroup=false
    Sampled Position Offset=$Particles.SpherePosition
    Sampled Position Offset Coordinate Space:ENiagaraMeshTransforms=::NewEnumerator3
    Skeletal Mesh=$User.SkeletalMesh
   ) {enabled=false, module=SkeletalMeshLocation002}
   &AddVelocity(Velocity=<0 0 -50>)
   &InheritVelocity(Module.Limit Speed=false, Velocity Scale=&VectorFromFloat(Value=1))
  ]
  particleUpdateScript [
   &ParticleState()
   &DynamicMaterialParameters(
    Index 0 Param 1=&RandomRangeFloat(Minimum=0, Maximum=1)
    Index 0 Param 2=1
    Module.Param0WriteEnabled=true
    Module.Param1WriteEnabled=true
    Module.Param2WriteEnabled=true
    Module.Param3WriteEnabled=true
   )
   &ScaleMeshSize(Scale Factor=&Multiply_VectorByFloat(
     Vector=&MakeVector(
      X=&Lerp_Float(A=0.25, B=1, Alpha=$Particles.NormalizedAge)
      Y=&Lerp_Float(A=1, B=0.5, Alpha=$Particles.NormalizedAge)
      Z=&Lerp_Float(A=1, B=0.5, Alpha=$Particles.NormalizedAge)
     )
     Float=&Lerp_Float(A=1, B=0, Alpha=$Particles.NormalizedAge)
    ))
   &GravityForce()
   &SolveForcesAndVelocity(Clamp Velocity=false, Limit Acceleration=false)
  ]
  rendererClasses [&NiagaraSpriteRendererProperties, &NiagaraMeshRendererProperties]
  renderers [
   {rendererClass=&NiagaraSpriteRendererProperties}
   {rendererClass=&NiagaraMeshRendererProperties, rendererIndex=1}
  ]
 }
}
summary {
 systemName=NS_Emote_Relaxed_HandBurst
 emitters [
  {emitterName=FlyingDrops, rendererClasses [&NiagaraSpriteRendererProperties, &NiagaraMeshRendererProperties]}
  {emitterName=FlyingDrops001, rendererClasses [&NiagaraSpriteRendererProperties, &NiagaraMeshRendererProperties]}
 ]
 userVariables [
  User.bIsFrontEndPreview {type=NiagaraBool}
  User.bIsSkydiving {type=NiagaraBool}
  User.ForwardDot {type=NiagaraFloat, defaultValue=0, description=""}
  User.IsSkydiving {type=NiagaraInt32}
  User.JumpTime {type=NiagaraFloat, defaultValue=0, description=""}
  User.ReactiveRate {type=NiagaraFloat, defaultValue=0, description=""}
  User.RunningSpeed {type=NiagaraFloat, defaultValue=0, description=""}
  User.SkeletalMesh {
   type=NiagaraDataInterfaceSkeletalMesh
   defaultValue {
    struct=NiagaraExt_VariableValue_DataInterface
    value {dataInterface=&self:NiagaraDataInterfaceSkeletalMesh_0}
   }
  }
  User.SpawnBurst {type=NiagaraBool}
 ]
}
